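# Niagara effect for the ThinGlaze character head cosmetic that sheds icing and colored sprinkles when the player sprints or lands.
# Both emitters spawn from SprintBurst and LandBurst user floats, suppressed while skydiving or head-hidden.
# A random, platform-weighted VisibilityTag varies particle size and behavior, and colors follow time of day.
# Sprinkles use height- and age-gated collision with per-tag bounce and friction, switching mesh, orientation and scale by speed.
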NiagaraSystem /BRCosmetics/Characters/Player/Female/Medium/Heads/F_MED_ThinGlaze_Head/FX/NS_Character_ThinGlaze_Idle_Icing
dependencies {
 usedDataInterfaces [
  &NiagaraDataInterfacePlatformSet
  &NiagaraDataInterfaceArrayColor
  &NiagaraDataInterfaceCurve
  &NiagaraDataInterfaceVector2DCurve
  &NiagaraDataInterfaceMeshRendererInfo
 ]
 usedDynamicInputs [&ClampFloat]
 usedModules [
  &SystemState
  &self:SystemSpawnScript.SystemScriptSource.SystemScriptGraph.NiagaraNodeAssignment_0.SetVariables_#1
 ]
}
emitters {
 Icing {
  emitterUpdateScript [
   &EmitterState(
    Enable Distance Culling=true
    Min Distance Response=::KillImmediately
    MinDistance=20
    Module.bUseMaxDistance=false
    Module.bUseMinDistance=true
    Scalability Mode=::NewEnumerator1
   )
   &SpawnRate(SpawnRate=50, Use Spawn Probability=false) {enabled=false, module=SpawnRate002}
   &SpawnRate(
    Spawn Probability=&MakeCustomFloatFromBool(
     Bool=$User.bIsHeadHidden
     True Float=0
     False Float=&MakeCustomFloatFromBool(Bool=$User.IsSkydiving, True Float=0, False Float=$User.SprintBurst)
    )
    SpawnRate=60
    Use Spawn Probability=true
   )
   &self:TriggerSpawn(
    Spawn Count=10
    Trigger:NiagaraBool=&Boolean_AndOperation(
     Bool1=&Boolean_NotOperation(Bool=$User.bIsHeadHidden)
     Bool2=&SetBoolByFloatComparison(A=&MakeCustomFloatFromBool(Bool=$User.IsSkydiving, True Float=0, False Float=$User.LandBurst), B=0.5)
    )
   ) {module=TriggerSpawn}
  ]
  particleSpawnScript [
   /SetVariables(Particles.VisibilityTag=&MakeIntFromFloat(Float=&UniformAOrBFloat(
      A=&UniformAOrBFloat(A=0, Evaluation Type=::NewEnumerator0, Recalculate Random Each Loop=true, Override Seed=false)
      B=&UniformAOrBFloat(
       A=2
       B=3
       Distribution Weight=&FN_PickFloatDependOnPlatform(TargetPlatform=DataInterface() {propertyValues {Platforms {qualityLevelMask=-29}}}, TargetPlatformActive=1, TargetPlatformNoActive=0.3)
       Evaluation Type=::NewEnumerator0
       Recalculate Random Each Loop=true
       Override Seed=false
      )
      Distribution Weight=0.3
      Evaluation Type=::NewEnumerator0
      Recalculate Random Each Loop=true
      Override Seed=false
     ))) {isSetParameters=true}
   &InitializeParticle(
    Color=&FNTimeOfDayColor(Night Color=<rgba 0.3971 0.6241 0.8513 1>, Morning Color=<rgba 1 1 1 1>, Day Color=<rgba 1 1 1 1>, Evening Color=<rgba 1 1 1 1>)
    Lifetime Max=0.3
    Lifetime Min=0.7
    Lifetime Mode=::NewEnumerator1
    Mesh Scale Mode=::NewEnumerator1
    Mesh Uniform Scale Max=0.1
    Mesh Uniform Scale Min=0.2
    Sprite Rotation Mode=::NewEnumerator2
    Sprite Size Mode=::NewEnumerator3
    Sprite UV Mode=::NewEnumerator3
    Uniform Sprite Size=&Multiply_Float(
     A=&MergeHelper_FloatSwitch(0=8, 1=17, 2=25, 3=?, Index=$Particles.VisibilityTag, Num=::NewEnumerator1)
     B=&RandomRangeFloat(Minimum=0.6, Maximum=1.2)
    )
    UsePositionOffset=false
    Write Material Random=true
   )
   &InheritVelocity(
    Inherited Velocity Amount Scale=<0.5 0.5 0.5>
    Inherited Velocity Speed Limit:NiagaraFloat=&Lerp_Float(A=400, B=750, Alpha=$User.SprintBurst)
    Speed Limit=true
    UseSpeedThreshold=false
   )
   &ShapeLocation(
    Disc Coverage=0.3
    MaskSpawnGroup=false
    Offset=&MakeCustomVectorFromBool(Bool=&SetBoolByFloatComparison(A=$User.SprintBurst, B=0.2), True Vector=<0 0 15>, False Vector=<0 0 -15>)
    Offset Mode=::NewEnumerator0
    Override Seed=false
    Ring Radius=35
    Rotation Angle=90
    Rotation Mode=::NewEnumerator1
    Shape Primitive=::NewEnumerator4
    UseOwnerScale=true
   )
   &AddVelocity(
    Cone Angle=120
    Cone Axis=<0 0 -1>
    Module.UseConeFalloff=false
    Module.UseVelDistribution=false
    Override Seed=false
    Rotation Coordinate Space=::Simulation
    Velocity Mode=::NewEnumerator2
    Velocity Speed=&RandomRangeFloat(Minimum=200, Maximum=100)
   )
   &SubUVAnimation(End Frame=3, SubUV Animation Mode:ENiagara_SubUVLookupMode=::NewEnumerator2)
   /SetVariables(
    Particles.SprinkleColor:LinearColor=&SelectLinearColorFromArray(
     Array Sampling Mode=::NewEnumerator0
     Direct Array Index=?
     Color Selection Array=DataInterface() {propertyValues {ColorData [<rgba 1 0.6986 0.0777 5>, <rgba 1 1 1 5>, <rgba 0.0564 0.1888 0.3328 5>]}}
     Random Seed=0
    )
    Particles.SprinkleSize:Vector2f=&RandomRangeVector2D(Minimum=<5 10>, Maximum=<2 6>)
   ) {enabled=false, isSetParameters=true}
  ]
  particleUpdateScript [
   &ParticleState()
   &ScaleMeshSize(Scale Factor=&Multiply_VectorByFloat(
     Vector=<1 1 1>
     Float=&FloatFromCurve(FloatCurve=DataInterface() {propertyValues {Curve {
         defaultValue=1
         keys [{interpMode=RCIM_Cubic, tangentMode=RCTM_User, time=0, value=1}, {interpMode=RCIM_Cubic, tangentMode=RCTM_User, time=1, value=0}]
        }}})
    )) {enabled=false}
   &GravityForce(Gravity=<0 0 -1200>)
   &Drag(Use Linear Drag=true, Use Rotational Drag=true)
   &ScaleSpriteSize(Scale Factor=&Vector2DFromCurve(
     Vector2Curve=DataInterface() {propertyValues {XCurve {keys [<at -0.0011 1.0067>,<at 0.9997 0.4319>]}, YCurve {keys [<at 0.0102 1.1102>,<at 0.0596 0.8185>,<at 0.9997 0.3098>]}}}
     ScaleCurve=<1 1>
    ))
   &SolveForcesAndVelocity(Clamp Velocity=false, Limit Acceleration=false)
   /SetVariables(Particles.Update Sprinkle Size:Vector2f=&Multiply_Vector2D_ByFloat(
     Float=&FloatFromCurve(FloatCurve=DataInterface() {propertyValues {Curve {
         defaultValue=1
         keys [{interpMode=RCIM_Cubic, tangentMode=RCTM_User, time=0, value=1}, {interpMode=RCIM_Cubic, tangentMode=RCTM_User, time=1, value=0}]
        }}})
     Vector2D=$Particles.SprinkleSize
    )) {enabled=false, isSetParameters=true}
  ]
  rendererClasses [&NiagaraSpriteRendererProperties, &NiagaraMeshRendererProperties]
  renderers [
   {rendererClass=&NiagaraSpriteRendererProperties}
   {rendererClass=&NiagaraSpriteRendererProperties, rendererIndex=1}
   {rendererClass=&NiagaraSpriteRendererProperties, rendererIndex=2}
   {rendererClass=&NiagaraMeshRendererProperties, rendererIndex=3}
   {rendererClass=&NiagaraSpriteRendererProperties, rendererIndex=4}
  ]
 }
 IcingSprinkles {
  emitterSpawnScript [/SetVariables(Emitter.RandomSeed:NiagaraInt32=/Niagara/DynamicInputs/Random/RandomInt(Randomness Mode:ENiagaraRandomnessMode=::NewEnumerator0, Random Seed:NiagaraInt32=0, Fixed Random Seed:NiagaraBool=false)) {isSetParameters=true}]
  emitterUpdateScript [
   &EmitterState(
    Enable Distance Culling=true
    Min Distance Response=::KillImmediately
    MinDistance=20
    Module.bUseMaxDistance=false
    Module.bUseMinDistance=true
    Scalability Mode=::NewEnumerator1
   )
   &SpawnRate(SpawnRate=50, Use Spawn Probability=false) {enabled=false, module=SpawnRate002}
   &SpawnRate(
    Spawn Probability=&MakeCustomFloatFromBool(
     Bool=$User.bIsHeadHidden
     True Float=0
     False Float=&MakeCustomFloatFromBool(Bool=$User.IsSkydiving, True Float=0, False Float=$User.SprintBurst)
    )
    SpawnRate=110
    Use Spawn Probability=true
   )
   &self:TriggerSpawn(
    Spawn Count:NiagaraInt32=&RandomRangeInt(Minimum=24, Maximum=25)
    Trigger:NiagaraBool=&Boolean_AndOperation(
     Bool1=&Boolean_NotOperation(Bool=$User.bIsHeadHidden)
     Bool2=&SetBoolByFloatComparison(A=&MakeCustomFloatFromBool(Bool=$User.IsSkydiving, True Float=0, False Float=$User.LandBurst), B=0.5)
    )
   ) {module=TriggerSpawn}
  ]
  particleSpawnScript [
   /SetVariables(Particles.VisibilityTag=&MakeIntFromFloat(Float=&UniformAOrBFloat(
      A=&MakeCustomFloatFromBool(
       Bool=&SetBoolByFloatComparison(
        A=&Abs_Float(Float:NiagaraFloat=&VectorLength(Vector=$Particles.Velocity))
        B=&Add_Float(A=15, B=30)
        Comparison Type=::NewEnumerator6
       )
       True Float=1
       False Float=0
      )
      B=&UniformAOrBFloat(
       A=2
       B=3
       Distribution Weight=&FN_PickFloatDependOnPlatform(TargetPlatform=DataInterface() {propertyValues {Platforms {qualityLevelMask=-29}}}, TargetPlatformActive=1, TargetPlatformNoActive=0.3)
       Evaluation Type=::NewEnumerator0
       Recalculate Random Each Loop=true
       Override Seed=false
      )
      Distribution Weight=0.85
      Evaluation Type=::NewEnumerator0
      Recalculate Random Each Loop=true
      Override Seed=false
     ))) {isSetParameters=true}
   &InitializeParticle(
    Color Mode=::NewEnumerator0
    Lifetime Max=1.5
    Lifetime Mode=::NewEnumerator1
    Mesh Scale=&MergeHelper_VectorSwitch(
     0=&Multiply_VectorByFloat(Vector=<1 1 0.5>, Float=&RandomRangeFloat(Minimum=0.3, Maximum=0.9))
     1=&Multiply_VectorByFloat(Vector=<1 1 0.5>, Float=&RandomRangeFloat(Minimum=0.3, Maximum=0.9))
     2=&Multiply_VectorByFloat(Vector=<1 1 1>, Float=&RandomRangeFloat(Minimum=0.2, Maximum=0.1))
     3=?
     Index=$Particles.VisibilityTag
     Num=::NewEnumerator1
    )
    Mesh Scale Mode=::NewEnumerator4
    UsePositionOffset=false
    Write Material Random=true
   )
   &InheritVelocity(
    Inherited Velocity Amount Scale=<0.5 0.5 0.5>
    Inherited Velocity Speed Limit:NiagaraFloat=&Lerp_Float(A=400, B=500, Alpha=$User.SprintBurst)
    Speed Limit=true
    UseSpeedThreshold=false
   )
   &ShapeLocation(
    Cylinder Height=1
    Cylinder Radius=35
    MaskSpawnGroup=false
    Offset Mode=::NewEnumerator0
    Override Seed=false
    Rotation Angle=90
    Rotation Mode=::NewEnumerator1
    Shape Primitive=::NewEnumerator1
    Surface Only=true
    Use Endcaps In Surface Only Mode=false
    UseOwnerScale=true
   ) {module=ShapeLocation001}
   &AddVelocity(
    Module.Use Falloff Distance=false
    Origin Offset=<10 0 0>
    Origin Offset Coordinate Space=::Simulation
    Override Seed=false
    Velocity Mode=::NewEnumerator1
    Velocity Speed=&RandomRangeFloat(Minimum=200, Maximum=100)
   ) {module=AddVelocity002}
   &InitialMeshOrientation(
    Mesh Orientation Mode=::NewEnumerator2
    Orientation Vector=<0 0 -1>
    Rotation=&RandomRangeVector(Minimum=<0 0 -1>, Maximum=<0 0 1>)
    Use Rotation Vector=true
   )
   /SetVariables(
    Particles.SprinkleColor:LinearColor=&MergeHelper_ColorSwitch(
     0=&FNTimeOfDayColor(
      Night Color=<rgba 0.3837 0.0979 0.358 5>
      Morning Color=<rgba 0.592 0.0634 0.5446 5>
      Day Color=<rgba 1 0.067 0.9162 5>
      Evening Color=<rgba 0.4485 0.0646 0.414 5>
     )
     1=&FNTimeOfDayColor(
      Night Color=<rgba 0.1425 0.179 0.1939 5>
      Morning Color=<rgba 0.2633 0.2633 0.2633 5>
      Day Color=<rgba 1 1 1 5>
      Evening Color=<rgba 0.4346 0.4346 0.4346 5>
     )
     2=&FNTimeOfDayColor(
      Night Color=<rgba 0.0135 0.1132 0.2216 5>
      Morning Color=<rgba 0.0247 0.2078 0.4068 5>
      Day Color=<rgba 0.0545 0.4584 0.8976 5>
      Evening Color=<rgba 0.0351 0.2953 0.5781 5>
     )
     3=&FNTimeOfDayColor(
      Night Color=<rgba 0.1107 0 0.4115 1>
      Morning Color=<rgba 0.1768 0 0.6568 1>
      Day Color=<rgba 0.2691 0 1 1>
      Evening Color=<rgba 0.1693 0 0.6291 1>
     )
     Index=&RandomRangeInt(Minimum=0, Maximum=3)
     Num=::NewEnumerator2
    )
    Particles.SprinkleSize:Vector2f=&RandomRangeVector2D(Minimum=<2 7>, Maximum=<2 5>)
   ) {isSetParameters=true}
  ]
  particleUpdateScript [
   &ParticleState()
   &GravityForce(Gravity=&MakeVector(X=0, Y=0, Z=&RandomRangeFloat(Minimum=-3500, Maximum=-1500)))
   &Collision(
    Age Colliding Particles=false
    CPU Collision Trace Channel=ECC_WorldStatic
    Collision Enabled=&Boolean_AndOperation(
     Bool1=&SetBoolByFloatComparison(
      A=&MakeFloatFromVector(Vector=Linked() {linkedVariable {name=Particles.Position, type=NiagaraPosition}}, Channel=::NewEnumerator2)
      B=&Add_Float(
       A=&MakeFloatFromVector(Vector=Linked() {linkedVariable {name=Engine.Owner.Position, type=NiagaraPosition}}, Channel=::NewEnumerator2)
       B=50
      )
      Comparison Type=::NewEnumerator6
     )
     Bool2=&SetBoolByFloatComparison(
      A=&FloatFromCurve(FloatCurve=DataInterface() {propertyValues {Curve {defaultValue=1, keys [<at 0.07 0 RCIM_Constant>,<at 0.07 0.9449 RCIM_Constant>]}}})
      B=0.5
     )
    )
    Friction:NiagaraFloat=&MergeHelper_FloatSwitch(
     0=&FloatFromCurve(
      FloatCurve=DataInterface() {propertyValues {Curve {keys [
          {interpMode=RCIM_Cubic, tangentMode=RCTM_User, time=0, value=0}
          {arriveTangent=0.2634, interpMode=RCIM_Cubic, leaveTangent=0.2634, tangentMode=RCTM_User, time=0.6, value=1}
         ]}}}
      Scale Curve=30
     )
     1=&FloatFromCurve(
      FloatCurve=DataInterface() {propertyValues {Curve {keys [
          {interpMode=RCIM_Cubic, tangentMode=RCTM_User, time=0, value=0}
          {arriveTangent=0.2634, interpMode=RCIM_Cubic, leaveTangent=0.2634, tangentMode=RCTM_User, time=0.6, value=1}
         ]}}}
      Scale Curve=30
     )
     2=0.1
     3=0.1
     Index=$Particles.VisibilityTag
     Num=::NewEnumerator2
    )
    KillOnCollision=false
    Max Number Of Collisions:NiagaraInt32=&MergeHelper_IntSwitch(0=3, 1=3, 2=200, 3=200, 4=?, 5=?, Index=$Particles.VisibilityTag, Num=::NewEnumerator2)
    Maximum Penetration Correction Distance=0
    Module.Control Rest State Via Surface Normal=false
    Module.EnableMaxCollisionCount=true
    Module.Randomize Collision Normal Vector=false
    Percentage of Penetration Before Rest=0
    Radius Calculation Type=::NewEnumerator3
    Rest State Time Range=0
    Restitution=&MergeHelper_FloatSwitch(
     0=0
     1=0
     2=&RandomRangeFloat(Minimum=0.5, Maximum=0.3)
     3=&RandomRangeFloat(Minimum=0.5, Maximum=0.3)
     Index=$Particles.VisibilityTag
     Num=::NewEnumerator2
    )
   )
   &SolveForcesAndVelocity(Clamp Velocity=false, Limit Acceleration=false)
   /SetVariables(Particles.MeshOrientation=/Niagara/DynamicInputs/TypeConversions/MakeCustomQuaternionFromBool(
     Bool:NiagaraBool=&SetBoolByFloatComparison(
      A=&Abs_Float(Float:NiagaraFloat=&VectorLength(Vector=$Particles.Velocity))
      B=15
      Comparison Type=::NewEnumerator6
     )
     True Quat:Quat4f=&OrientMeshToVector(Look At Direction=$Output.Collision.CombinedCollisionNormal, Reference Axis Vector=<1 0 0>)
     False Quat:Quat4f=&MakeQuatFromAxisAngle(Axis=<1 0 0>, AngleInDegrees=&RandomRangeFloat(Minimum=-360, Maximum=360))
    )) {isSetParameters=true}
   &ScaleMeshSize(Scale Factor=&MakeCustomVectorFromBool(
     Bool=&SetBoolByFloatComparison(
      A=&Abs_Float(Float:NiagaraFloat=&VectorLength(Vector=$Particles.Velocity))
      B=15
      Comparison Type=::NewEnumerator6
     )
     True Vector=&RandomRangeVector(Minimum=<1 1 1.8>, Maximum=<1.4 1.4 3.5>)
     False Vector=<1 1 1>
    )) {module=ScaleMeshSize001}
   &ScaleMeshSize(Scale Factor=&VectorFromFloat(Value=&FloatFromCurve(FloatCurve=DataInterface() {propertyValues {Curve {keys [
          {interpMode=RCIM_Cubic, tangentMode=RCTM_User, time=0, value=0.6}
          {interpMode=RCIM_Cubic, tangentMode=RCTM_User, time=0.2, value=1}
          {interpMode=RCIM_Cubic, tangentMode=RCTM_User, time=0.85, value=1}
          {interpMode=RCIM_Cubic, tangentMode=RCTM_User, time=1, value=0}
         ]}}})))
   &ScaleMeshSizeBySpeed(Max Scale Factor=&RandomRangeVector(Minimum=<2 1.2 1.2>, Maximum=<3 1.5 1.5>), Speed Threshold=570)
   /SetVariables(Particles.MeshIndex=&MakeCustomIntFromBool(
     Bool=&SetBoolByFloatComparison(
      A=&Abs_Float(Float:NiagaraFloat=&VectorLength(Vector=$Particles.Velocity))
      B=15
      Comparison Type=::NewEnumerator6
     )
     True INT=&RandomRangeInt(Minimum=0, Maximum=1)
     False INT=&RandomRangeInt(Minimum=0, Maximum=3)
    )) {isSetParameters=true}
   /SetVariables(Particles.Update Sprinkle Size:Vector2f=&Multiply_Vector2D_ByFloat(
     Float=&FloatFromCurve(FloatCurve=DataInterface() {propertyValues {Curve {
         defaultValue=1
         keys [{interpMode=RCIM_Cubic, tangentMode=RCTM_User, time=0, value=1}, {interpMode=RCIM_Cubic, tangentMode=RCTM_User, time=1, value=0}]
        }}})
     Vector2D=$Particles.SprinkleSize
    )) {isSetParameters=true}
   &ScaleMeshSize(Scale Factor=&Multiply_VectorByFloat(
     Vector=<1 1 1>
     Float=&MergeHelper_FloatSwitch(
      0=1
      1=1
      2=&FloatFromCurve(FloatCurve=DataInterface() {propertyValues {Curve {
          defaultValue=1
          keys [{interpMode=RCIM_Cubic, tangentMode=RCTM_User, time=0, value=1}, {interpMode=RCIM_Cubic, tangentMode=RCTM_User, time=1, value=0}]
         }}})
      3=?
      Index=$Particles.VisibilityTag
      Num=::NewEnumerator1
     )
    )) {module=ScaleMeshSize002}
  ]
  rendererClasses [&NiagaraMeshRendererProperties, &NiagaraSpriteRendererProperties]
  renderers [
   {rendererClass=&NiagaraMeshRendererProperties, rendererIndex=0}
   {rendererClass=&NiagaraMeshRendererProperties, rendererIndex=1}
   {rendererClass=&NiagaraSpriteRendererProperties, rendererIndex=2}
   {rendererClass=&NiagaraMeshRendererProperties, rendererIndex=3}
  ]
 }
}
summary {
 systemName=NS_Character_ThinGlaze_Idle_Icing
 emitters [
  {emitterName=Icing, rendererClasses [&NiagaraSpriteRendererProperties, &NiagaraMeshRendererProperties]}
  {emitterName=IcingSprinkles, rendererClasses [&NiagaraMeshRendererProperties, &NiagaraSpriteRendererProperties]}
 ]
 userVariables [
  User.bIsFrontEndPreview {type=NiagaraBool}
  User.bIsHeadHidden {type=NiagaraBool}
  User.IsSkydiving {type=NiagaraBool}
  User.LandBurst {type=NiagaraFloat, defaultValue=0, description=""}
  User.SprintBurst {type=NiagaraFloat, defaultValue=0, description=""}
 ]
}
